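# Physics asset for the collar of the male medium MusketSlinger cosmetic body, set up for secondary motion on the dyn_collar bones.
# Two thin boxes form the left and right collar flaps, three small capsules cover the back, and larger head and clavicle capsules serve as apparent colliders.
# Each collar bone has limited swing and twist against its MSTR root.
# Those roots constrain with free angular motion to spine_05 for the flaps and neck_01 for the back.
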PhysicsAsset /BRCosmetics/Characters/Player/Male/Medium/Bodies/M_MED_MusketSlinger/Meshes/M_MED_MusketSlinger_Collar_Physics
bodies {
 clavicle_l [{shapeType=Capsule, length=6, radius=7.7, shapeName=clavicle_l_capsule, center=<10.4323 3.5 -2.3691>, rotation=<rot -70 179.9995 -179.9996>}]
 clavicle_r [{shapeType=Capsule, length=6, radius=7.7, shapeName=clavicle_r_capsule, center=<-10.4323 -3.5 2.3691>, rotation=<rot -70 179.9994 -179.9995>}]
 dyn_collar_01_L [{
   shapeType=Box
   extentX=7.0276
   extentY=0.5
   extentZ=16.4616
   shapeName=dyn_collar_01_L_box
   center=<9.4994 0 -0.0546>
   rotation=<rot 0 -84.4528 -86.7594>
  }]
 dyn_collar_01_R [{
   shapeType=Box
   extentX=6.4939
   extentY=0.5
   extentZ=15.8745
   shapeName=dyn_collar_01_R_box
   center=<9.0959 0 0.1608>
   rotation=<rot 0 82.6352 -89.3466>
  }]
 dyn_collar_MSTR_L [{shapeType=Capsule, length=0, radius=0.5, shapeName=dyn_collar_MSTR_L_capsule, rotation=<rot -80.0004 179.9983 -179.9967>}]
 dyn_collar_MSTR_R [{shapeType=Capsule, length=0, radius=0.5, shapeName=dyn_collar_MSTR_R_capsule, rotation=<rot -80.0004 -179.9982 179.9967>}]
 dyn_collar_bk_01 [{shapeType=Capsule, length=5, radius=1, shapeName=dyn_collar_bk_01_capsule, center=<3.6106 0.0382 0>, rotation=<rot -90 180 92.7125>}]
 dyn_collar_bk_01_L [{shapeType=Capsule, length=5, radius=1, shapeName=dyn_collar_bk_01_L_capsule, center=<4.3778 0.2773 0>, rotation=<rot -80 100.0005 -179.9997>}]
 dyn_collar_bk_01_R [{shapeType=Capsule, length=5, radius=1, shapeName=dyn_collar_bk_01_R_capsule, center=<4.5349 -0.6296 0>, rotation=<rot -90 90 159.9993>}]
 dyn_collar_bk_MSTR [{shapeType=Capsule, length=0, radius=0.5, shapeName=dyn_collar_bk_MSTR_capsule, rotation=<rot -90 2.7126 -2.7126>}]
 dyn_collar_bk_MSTR_L [{shapeType=Capsule, length=0, radius=0.5, shapeName=dyn_collar_bk_MSTR_L_capsule, rotation=<rot -90 36.6528 -36.6522>}]
 dyn_collar_bk_MSTR_R [{shapeType=Capsule, length=0, radius=0.5, shapeName=dyn_collar_bk_MSTR_R_capsule, rotation=<rot -90 -36.6528 36.6523>}]
 head [
  {shapeType=Capsule, length=7, radius=4.5, shapeName=head_capsule, center=<-0.1083 0 0>, rotation=<rot 0 179.9995 180>}
  {shapeType=Capsule, length=6, radius=5, shapeName=head_capsule, center=<0.7657 6.2745 0>, rotation=<rot 90 -179.9996 -179.9993>}
 ]
 neck_01 [{shapeType=Capsule, length=0, radius=0.5, shapeName=neck_01_capsule}]
 spine_05 [{shapeType=Capsule, length=0, radius=0.5, shapeName=spine_05_capsule}]
}
constraints [
 {bone1Name=dyn_collar_bk_01_R, bone2Name=dyn_collar_bk_MSTR_R, swing1LimitDegrees=20, swing2LimitDegrees=20, twistLimitDegrees=5}
 {bone1Name=dyn_collar_bk_01_L, bone2Name=dyn_collar_bk_MSTR_L, swing1LimitDegrees=20, swing2LimitDegrees=20, twistLimitDegrees=5}
 {bone1Name=dyn_collar_bk_01, bone2Name=dyn_collar_bk_MSTR, swing1LimitDegrees=20, swing2LimitDegrees=20, twistLimitDegrees=5}
 {bone1Name=dyn_collar_01_L, bone2Name=dyn_collar_MSTR_L, swing1LimitDegrees=5, swing2LimitDegrees=45, twistLimitDegrees=5}
 {bone1Name=dyn_collar_01_R, bone2Name=dyn_collar_MSTR_R, swing1LimitDegrees=5, swing2LimitDegrees=45, twistLimitDegrees=5}
 {
  bone1Name=dyn_collar_MSTR_L
  bone2Name=spine_05
  swing1LimitDegrees=45
  swing1Motion=Free
  swing2LimitDegrees=45
  swing2Motion=Free
  twistLimitDegrees=45
  twistMotion=Free
 }
 {
  bone1Name=dyn_collar_MSTR_R
  bone2Name=spine_05
  swing1LimitDegrees=45
  swing1Motion=Free
  swing2LimitDegrees=45
  swing2Motion=Free
  twistLimitDegrees=45
  twistMotion=Free
 }
 {
  bone1Name=dyn_collar_bk_MSTR
  bone2Name=neck_01
  swing1LimitDegrees=45
  swing1Motion=Free
  swing2LimitDegrees=45
  swing2Motion=Free
  twistLimitDegrees=45
  twistMotion=Free
 }
 {
  bone1Name=dyn_collar_bk_MSTR_L
  bone2Name=neck_01
  swing1LimitDegrees=45
  swing1Motion=Free
  swing2LimitDegrees=45
  swing2Motion=Free
  twistLimitDegrees=45
  twistMotion=Free
 }
 {
  bone1Name=dyn_collar_bk_MSTR_R
  bone2Name=neck_01
  swing1LimitDegrees=45
  swing1Motion=Free
  swing2LimitDegrees=45
  swing2Motion=Free
  twistLimitDegrees=45
  twistMotion=Free
 }
]
