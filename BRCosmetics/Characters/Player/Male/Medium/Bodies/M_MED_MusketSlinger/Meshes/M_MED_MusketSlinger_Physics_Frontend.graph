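# Frontend physics asset for the male medium MusketSlinger body cosmetic, set up for secondary motion on the jacket.
# Seven jacket chains (left, right, three back, two front) each link two capsule segments to a tiny master anchor body.
# Constraints limit twist and one swing axis tightly while allowing a wider second swing.
# Unconstrained capsules on the pelvis, lower spine and thighs provide body shapes for the jacket to collide against.
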PhysicsAsset /BRCosmetics/Characters/Player/Male/Medium/Bodies/M_MED_MusketSlinger/Meshes/M_MED_MusketSlinger_Physics_Frontend
bodies {
 dyn_jacket_01_L [{shapeType=Capsule, length=8.5, radius=2, shapeName=dyn_jacket_01_L_capsule, center=<6.7241 0.0001 0>, rotation=<rot -90 -165.9638 75.9644>}]
 dyn_jacket_01_R [{shapeType=Capsule, length=8.5, radius=2, shapeName=dyn_jacket_01_R_capsule, center=<6.7241 0.0001 0>, rotation=<rot -90 -61.4916 -28.509>}]
 dyn_jacket_02_L [{shapeType=Capsule, length=12.5, radius=2, shapeName=dyn_jacket_02_L_capsule, center=<6.3501 0 0>, rotation=<rot -90 -165.9638 75.9644>}]
 dyn_jacket_02_R [{
   shapeType=Capsule
   length=12.5
   radius=2.7
   shapeName=dyn_jacket_02_R_capsule
   center=<6.3501 -0.0001 0>
   rotation=<rot -90 153.4349 116.5644>
  }]
 dyn_jacket_MSTR_L [{shapeType=Capsule, length=0, radius=0.5, shapeName=dyn_jacket_MSTR_L_capsule, rotation=<rot -90 61.4916 -61.4912>}]
 dyn_jacket_MSTR_R [{shapeType=Capsule, length=0, radius=0.5, shapeName=dyn_jacket_MSTR_R_capsule, rotation=<rot -90 -61.4916 61.4912>}]
 dyn_jacket_bk_01 [{shapeType=Capsule, length=8, radius=2.5, shapeName=dyn_jacket_bk_01_capsule, center=<4.0411 0 0>, rotation=<rot -90 179.3441 90.6559>}]
 dyn_jacket_bk_01_L [{shapeType=Capsule, length=8, radius=2.5, shapeName=dyn_jacket_bk_01_L_capsule, center=<4.0411 0 0>, rotation=<rot -90 172.6904 97.3101>}]
 dyn_jacket_bk_01_R [{shapeType=Capsule, length=8, radius=2.5, shapeName=dyn_jacket_bk_01_R_capsule, center=<4.0411 0 0>, rotation=<rot -90 -172.6904 82.6899>}]
 dyn_jacket_bk_02 [{shapeType=Capsule, length=11, radius=3, shapeName=dyn_jacket_bk_02_capsule, center=<6.3501 0 0>, rotation=<rot -90 180 90>}]
 dyn_jacket_bk_02_L [{shapeType=Capsule, length=8, radius=2.5, shapeName=dyn_jacket_bk_02_L_capsule, center=<6.3501 0 0>, rotation=<rot -90 180 90.0005>}]
 dyn_jacket_bk_02_R [{shapeType=Capsule, length=8, radius=2.5, shapeName=dyn_jacket_bk_02_R_capsule, center=<6.3501 0 0>, rotation=<rot -90 180 89.9995>}]
 dyn_jacket_bk_MSTR [{shapeType=Capsule, length=0, radius=0.5, shapeName=dyn_jacket_bk_MSTR_capsule, rotation=<rot -90 179.3441 -179.3441>}]
 dyn_jacket_bk_MSTR_L [{shapeType=Capsule, length=0, radius=0.5, shapeName=dyn_jacket_bk_MSTR_L_capsule, rotation=<rot -90 172.6904 -172.6901>}]
 dyn_jacket_bk_MSTR_R [{shapeType=Capsule, length=0, radius=0.5, shapeName=dyn_jacket_bk_MSTR_R_capsule, rotation=<rot -90 -172.6904 172.6901>}]
 dyn_jacket_fr_01_L [{shapeType=Capsule, length=8.5, radius=1.7, shapeName=dyn_jacket_fr_01_L_capsule, center=<6.5534 0 0>, rotation=<rot -90 180 90.0004>}]
 dyn_jacket_fr_01_R [{shapeType=Capsule, length=8.5, radius=1.7, shapeName=dyn_jacket_fr_01_R_capsule, center=<6.6621 0 0>, rotation=<rot -90 -165.9638 75.9634>}]
 dyn_jacket_fr_02_L [{
   shapeType=Capsule
   length=12
   radius=2.1
   shapeName=dyn_jacket_fr_02_L_capsule
   center=<6.3501 -3.0742 0>
   rotation=<rot -80 -89.9993 -0.0002>
  }]
 dyn_jacket_fr_02_R [{
   shapeType=Capsule
   length=12
   radius=2.1
   shapeName=dyn_jacket_fr_02_R_capsule
   center=<6.3501 3.0742 0>
   rotation=<rot -80 89.9993 -179.9997>
  }]
 dyn_jacket_fr_MSTR_L [{shapeType=Capsule, length=0, radius=0.5, shapeName=dyn_jacket_fr_MSTR_L_capsule, rotation=<rot -90 10.3923 -10.3921>}]
 dyn_jacket_fr_MSTR_R [{shapeType=Capsule, length=0, radius=0.5, shapeName=dyn_jacket_fr_MSTR_R_capsule, rotation=<rot -90 -10.3923 10.3921>}]
 pelvis [{shapeType=Capsule, length=8.5, radius=9.5, shapeName=pelvis_capsule}]
 spine_01 [{shapeType=Capsule, length=8.5, radius=8, shapeName=spine_01_capsule}]
 spine_02 [{shapeType=Capsule, length=4.5, radius=10, shapeName=spine_02_capsule, rotation=<rot 90 0 0>}]
 thigh_l [{shapeType=Capsule, length=19, radius=7.8, shapeName=thigh_l_capsule, center=<-20.733 0.0001 1.427>, rotation=<rot 90 89.6476 89.6487>}]
 thigh_r [{shapeType=Capsule, length=18, radius=9, shapeName=thigh_r_capsule, center=<20.8085 0.1454 0>, rotation=<rot 90 89.6476 89.6487>}]
}
constraints [
 {bone1Name=dyn_jacket_01_L, bone2Name=dyn_jacket_MSTR_L, swing1LimitDegrees=10, swing2LimitDegrees=45, twistLimitDegrees=5}
 {bone1Name=dyn_jacket_02_L, bone2Name=dyn_jacket_01_L, swing1LimitDegrees=10, swing2LimitDegrees=45, twistLimitDegrees=5}
 {bone1Name=dyn_jacket_01_R, bone2Name=dyn_jacket_MSTR_R, swing1LimitDegrees=10, swing2LimitDegrees=45, twistLimitDegrees=5}
 {bone1Name=dyn_jacket_02_R, bone2Name=dyn_jacket_01_R, swing1LimitDegrees=10, swing2LimitDegrees=45, twistLimitDegrees=5}
 {bone1Name=dyn_jacket_bk_01, bone2Name=dyn_jacket_bk_MSTR, swing1LimitDegrees=2.5, swing2LimitDegrees=45, twistLimitDegrees=2.5}
 {bone1Name=dyn_jacket_bk_02, bone2Name=dyn_jacket_bk_01, swing1LimitDegrees=15, swing2LimitDegrees=45, twistLimitDegrees=2.5}
 {bone1Name=dyn_jacket_bk_01_L, bone2Name=dyn_jacket_bk_MSTR_L, swing1LimitDegrees=10, swing2LimitDegrees=45, twistLimitDegrees=5}
 {bone1Name=dyn_jacket_bk_02_L, bone2Name=dyn_jacket_bk_01_L, swing1LimitDegrees=15, swing2LimitDegrees=45, twistLimitDegrees=2.2}
 {bone1Name=dyn_jacket_bk_01_R, bone2Name=dyn_jacket_bk_MSTR_R, swing1LimitDegrees=10, swing2LimitDegrees=45, twistLimitDegrees=5}
 {bone1Name=dyn_jacket_bk_02_R, bone2Name=dyn_jacket_bk_01_R, swing1LimitDegrees=15, swing2LimitDegrees=45, twistLimitDegrees=2.5}
 {bone1Name=dyn_jacket_fr_01_L, bone2Name=dyn_jacket_fr_MSTR_L, swing1LimitDegrees=10, swing2LimitDegrees=45, twistLimitDegrees=5}
 {bone1Name=dyn_jacket_fr_02_L, bone2Name=dyn_jacket_fr_01_L, swing1LimitDegrees=10, swing2LimitDegrees=45, twistLimitDegrees=5}
 {bone1Name=dyn_jacket_fr_01_R, bone2Name=dyn_jacket_fr_MSTR_R, swing1LimitDegrees=10, swing2LimitDegrees=45, twistLimitDegrees=5}
 {bone1Name=dyn_jacket_fr_02_R, bone2Name=dyn_jacket_fr_01_R, swing1LimitDegrees=10, swing2LimitDegrees=45, twistLimitDegrees=5}
]
