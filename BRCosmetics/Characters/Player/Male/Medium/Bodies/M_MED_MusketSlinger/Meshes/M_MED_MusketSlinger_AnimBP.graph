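# Post-process style AnimBP for the MusketSlinger male medium body, adding secondary motion to jacket, collar and gear.
# Copied parent pose runs through Control Rig, look-ats, constraints and corrective bone offsets into a cached pose.
# Frontend or in-game RigidBody simulates that pose, then collar physics and collar correctives follow.
# EventGraph derives corrective alphas from bone deltas, adjusts for skydiving and crouching, and drives WPO cloth material wind, drag and turbulence.
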AnimBlueprint /BRCosmetics/Characters/Player/Male/Medium/Bodies/M_MED_MusketSlinger/Meshes/M_MED_MusketSlinger_AnimBP
target_skeleton=&/M_MED_MusketSlinger_Skeleton
graphs [
 AnimGraph {nodes [
   n1 = Get Head Pitch() -> (Head Pitch:float)
   n2 = RigidBody!(
    ComponentPose=Local To Component(LocalPose=Use cached pose 'Jacket Cashe'())
    Alpha
    OverrideWorldGravity:Vector=Get Gravity Override()
    ComponentLinearAccScale:Vector=Get ComponentLinearAccScale() -> (ComponentLinearAccScale:Vector)
    ComponentLinearVelScale:Vector=Get ComponentLinearVelScale() -> (ComponentLinearVelScale:Vector)
   ) -> (Pose) {
    bOverrideWorldGravity=true
    overridePhysicsAsset=&/M_MED_MusketSlinger_Physics_Frontend
    simulationSpace=BaseBoneSpace
    alphaBoolBlend {blendOption=Linear}
    baseBoneRef {boneName=pelvis}
    componentAppliedLinearAccClamp=<1000 1000 1000>
    componentLinearAccScale=<0.4 0.4 0.4>
    componentLinearVelScale=<0.4 0.4 0.4>
    overrideWorldGravity {z=-980}
   }
   n3 = Component To Local(ComponentPose=@n2)
   n4 = RigidBody(
    ComponentPose=Local To Component(LocalPose=Use cached pose 'Jacket Cashe'())
    OverrideWorldGravity=Get Gravity Override()
    ComponentLinearAccScale=Get ComponentLinearAccScale()
    ComponentLinearVelScale=Get ComponentLinearVelScale()
   ) {
    bOverrideWorldGravity=true
    overridePhysicsAsset=&/M_MED_MusketSlinger_Physics
    simulationSpace=BaseBoneSpace
    alphaBoolBlend {blendOption=Linear}
    baseBoneRef {boneName=pelvis}
    componentAppliedLinearAccClamp=<1000 1000 1000>
    componentLinearAccScale=<0.4 0.4 0.4>
    componentLinearVelScale=<0.4 0.4 0.4>
    overrideWorldGravity {z=-980}
   }
   n5 = Component To Local(ComponentPose=@n4)
   n6 = Get bIsInFrontEnd()
   n7 = RigidBody!(
    ComponentPose=Local To Component(LocalPose=Blend Poses by bool(BlendPose_0=@n3, BlendPose_1=@n5, bActiveValue=@n6) {blendType=Linear})
    Alpha
    OverrideWorldGravity=Get Gravity Override()
   ) -> (Pose) {
    bOverrideWorldGravity=true
    overridePhysicsAsset=&/M_MED_MusketSlinger_Collar_Physics
    simulationSpace=BaseBoneSpace
    alphaBoolBlend {blendOption=Linear}
    baseBoneRef {boneName=spine_05}
    componentAppliedLinearAccClamp=<1000 1000 1000>
    componentLinearAccScale=<0.4 0.4 0.4>
    componentLinearVelScale=<0.4 0.4 0.4>
    overrideWorldGravity {z=-980}
   }
   n8 = "Transform (Modify) Bone - Bone: base_collar_bk"(ComponentPose=@n7, Alpha=Get Neck 01 Rotation L() -> (Neck 01 Rotation L:double), Translation="0,-0.5,0", Rotation="0,0,27", Scale="1,1,1") {
    rotationMode=BMM_Additive
    rotationSpace=BCS_BoneSpace
    alphaBoolBlend {blendOption=Linear}
    boneToModify {boneName=base_collar_bk}
    rotation {roll=27}
    translation {y=-0.5}
   }
   n9 = "Transform (Modify) Bone - Bone: base_collar_bk"(ComponentPose=@n8, Alpha=Get Neck 01 Rotation R() -> (Neck 01 Rotation R:float), Translation="0,-0.5,0", Rotation="0,0,-27", Scale="1,1,1") {
    rotationMode=BMM_Additive
    rotationSpace=BCS_BoneSpace
    alphaBoolBlend {blendOption=Linear}
    boneToModify {boneName=base_collar_bk}
    rotation {roll=-27}
    translation {y=-0.5}
   }
   n10 = "Transform (Modify) Bone - Bone: dyn_collar_bk_MSTR_L"(
    ComponentPose=@n9
    Alpha=Get Clavicle L upward movement() -> (Clavicle L upward movement:double)
    Translation="0,0,0"
    Rotation=Get Collar BK MSTR L Upper Rotation() -> (Collar BK MSTR L Upper Rotation:Rotator)
    Scale="1,1,1"
   ) {
    rotationMode=BMM_Additive
    rotationSpace=BCS_BoneSpace
    translationMode=BMM_Additive
    translationSpace=BCS_BoneSpace
    alphaBoolBlend {blendOption=Linear}
    boneToModify {boneName=dyn_collar_bk_MSTR_L}
    rotation=<rot -13.883 8.1061 -26.9711>
   }
   n11 = "Transform (Modify) Bone - Bone: dyn_collar_bk_MSTR_R"(
    ComponentPose=@n10
    Alpha=Get Clavicle R upward movement() -> (Clavicle R upward movement:float)
    Translation="0,0,0"
    Rotation=Get Collar BK MSTR R Upper Rotation() -> (Collar BK MSTR R Upper Rotation:Rotator)
    Scale="1,1,1"
   ) {
    rotationMode=BMM_Additive
    rotationSpace=BCS_BoneSpace
    translationMode=BMM_Additive
    translationSpace=BCS_BoneSpace
    alphaBoolBlend {blendOption=Linear}
    boneToModify {boneName=dyn_collar_bk_MSTR_R}
    rotation=<rot -12.745 -3.542 47.522>
   }
   n12 = "Transform (Modify) Bone - Bone: dyn_collar_bk_MSTR"(ComponentPose=@n11, Alpha=@n1, Translation="0,0,1", Rotation="0,0,0", Scale="1,1,1") {
    rotationMode=BMM_Additive
    rotationSpace=BCS_BoneSpace
    translationMode=BMM_Additive
    translationSpace=BCS_BoneSpace
    alphaBoolBlend {blendOption=Linear}
    alphaScaleBiasClamp {interpSpeedDecreasing=0.5, interpSpeedIncreasing=0.5}
    boneToModify {boneName=dyn_collar_bk_MSTR}
    translation {z=1}
   }
   n13 = "Transform (Modify) Bone - Bone: dyn_collar_bk_MSTR_L"(ComponentPose=@n12, Alpha=@n1, Translation="0,0,1", Rotation="0,0,0", Scale="1,1,1") {
    rotationMode=BMM_Additive
    rotationSpace=BCS_BoneSpace
    translationMode=BMM_Additive
    translationSpace=BCS_BoneSpace
    alphaBoolBlend {blendOption=Linear}
    alphaScaleBiasClamp {interpSpeedDecreasing=0.5, interpSpeedIncreasing=0.5}
    boneToModify {boneName=dyn_collar_bk_MSTR_L}
    translation {z=1}
   }
   n14 = "Transform (Modify) Bone - Bone: dyn_collar_bk_MSTR_R"(ComponentPose=@n13, Alpha=@n1, Translation="0,0,1", Rotation="0,0,0", Scale="1,1,1") {
    rotationMode=BMM_Additive
    rotationSpace=BCS_BoneSpace
    translationMode=BMM_Additive
    translationSpace=BCS_BoneSpace
    alphaBoolBlend {blendOption=Linear}
    alphaScaleBiasClamp {interpSpeedDecreasing=0.5, interpSpeedIncreasing=0.5}
    boneToModify {boneName=dyn_collar_bk_MSTR_R}
    translation {z=1}
   }
   Output Pose(Result=Component To Local(ComponentPose=@n14)) {layerGroup=None}
   n15 = Get Buckle Side Rotation() -> (Buckle Side Rotation:Rotator)
   n16 = Get body rotation to the left() -> (body rotation to the left:double)
   n17 = Get Bag correction when crouching() -> (Bag correction when crouching:double)
   n18 = Control Rig(Source=Copy Pose From Mesh() {bUseAttachedParent=true}) {
    alphaBoolBlend {blendOption=Linear}
    controlRigAssetReference {blueprintRigClass=&/M_MED_MusketSlinger_ControlRig.M_MED_MusketSlinger_ControlRig_C}
   }
   n19 = "Constraint - dyn_jacket_fr_CST_L (P)"(ComponentPose=Local To Component(LocalPose=@n18), Alpha=Get Crouching standing() -> (Crouching standing:double), ConstraintWeights_0=1) {
    alphaBoolBlend {blendOption=Linear}
    boneToModify {boneName=dyn_jacket_fr_CST_L}
    constraintSetup [{targetBone {boneName=pelvis}}]
    constraintWeights [1]
   }
   n20 = Look At - Bone: dyn_jacket_fr_MSTR_L(ComponentPose=@n19, Alpha=1) {
    alphaBoolBlend {blendOption=Linear}
    boneToModify {boneName=dyn_jacket_fr_MSTR_L}
    lookAtLocation {x=0}
    lookAtTarget {boneReference {boneName=calf_l}}
    lookAt_Axis {axis=<1 0>}
   }
   n21 = Look At - Bone: dyn_jacket_MSTR_L(ComponentPose=@n20, Alpha=0.8) {
    alphaBoolBlend {blendOption=Linear}
    boneToModify {boneName=dyn_jacket_MSTR_L}
    lookAtLocation {x=0}
    lookAtTarget {boneReference {boneName=calf_l}}
    lookAt_Axis {axis=<1 0>}
   }
   n22 = Look At - Bone: dyn_jacket_fr_MSTR_R(ComponentPose=@n21, Alpha=0.8) {
    alphaBoolBlend {blendOption=Linear}
    boneToModify {boneName=dyn_jacket_fr_MSTR_R}
    lookAtLocation {x=0}
    lookAtTarget {boneReference {boneName=calf_r}}
    lookAt_Axis {axis=<1 0>}
   }
   n23 = Look At - Bone: dyn_jacket_MSTR_R(ComponentPose=@n22, Alpha=0.7) {
    alphaBoolBlend {blendOption=Linear}
    boneToModify {boneName=dyn_jacket_MSTR_R}
    lookAtLocation {x=0}
    lookAtTarget {boneReference {boneName=calf_r}}
    lookAt_Axis {axis=<1 0>}
   }
   n24 = "Constraint - bag_01_L (PPPP)"!(
    ComponentPose=@n23
    Alpha=1
    ConstraintWeights_0=0.0777
    ConstraintWeights_1:float=0.201
    ConstraintWeights_2:float=0.5593
    ConstraintWeights_3:float=0.162
   ) -> (Pose) {
    alphaBoolBlend {blendOption=Linear}
    boneToModify {boneName=bag_01_L}
    constraintSetup [{targetBone {boneName=spine_04}}, {targetBone {boneName=spine_01}}, {targetBone {boneName=spine_02}}, {targetBone {boneName=spine_03}}]
    constraintWeights [1,1,1,1]
   }
   n25 = "Constraint - bag_02_L (PPPP)"!(
    ComponentPose=@n24
    Alpha=1
    ConstraintWeights_0=0.0777
    ConstraintWeights_1:float=0.201
    ConstraintWeights_2:float=0.5593
    ConstraintWeights_3:float=0.162
   ) -> (Pose) {
    alphaBoolBlend {blendOption=Linear}
    boneToModify {boneName=bag_02_L}
    constraintSetup [{targetBone {boneName=spine_04}}, {targetBone {boneName=spine_01}}, {targetBone {boneName=spine_02}}, {targetBone {boneName=spine_03}}]
    constraintWeights [1,1,1,1]
   }
   n26 = "Constraint - bag_01_R (PPPP)"!(
    ComponentPose=@n25
    Alpha=1
    ConstraintWeights_0=0.0777
    ConstraintWeights_1:float=0.201
    ConstraintWeights_2:float=0.5593
    ConstraintWeights_3:float=0.162
   ) -> (Pose) {
    alphaBoolBlend {blendOption=Linear}
    boneToModify {boneName=bag_01_R}
    constraintSetup [{targetBone {boneName=spine_04}}, {targetBone {boneName=spine_01}}, {targetBone {boneName=spine_02}}, {targetBone {boneName=spine_03}}]
    constraintWeights [1,1,1,1]
   }
   n27 = "Constraint - bag_fr (PPPP)"!(
    ComponentPose=@n26
    Alpha=1
    ConstraintWeights_0=0.3539
    ConstraintWeights_1:float=0.3367
    ConstraintWeights_2:float=0.0428
    ConstraintWeights_3:float=0.2666
   ) -> (Pose) {
    alphaBoolBlend {blendOption=Linear}
    boneToModify {boneName=bag_fr}
    constraintSetup [{targetBone {boneName=spine_02}}, {targetBone {boneName=spine_03}}, {targetBone {boneName=vest_fr_05}}, {targetBone {boneName=vest_fr_04}}]
    constraintWeights [1,1,1,1]
   }
   n28 = "Constraint - radio (PPPP)"!(
    ComponentPose=@n27
    Alpha=1
    ConstraintWeights_0=0.0777
    ConstraintWeights_1:float=0.201
    ConstraintWeights_2:float=0.5593
    ConstraintWeights_3:float=0.162
   ) -> (Pose) {
    alphaBoolBlend {blendOption=Linear}
    boneToModify {boneName=radio}
    constraintSetup [{targetBone {boneName=spine_04}}, {targetBone {boneName=spine_01}}, {targetBone {boneName=spine_02}}, {targetBone {boneName=spine_03}}]
    constraintWeights [1,1,1,1]
   }
   n29 = "Transform (Modify) Bone - Bone: sleeve_L"(ComponentPose=@n28, Alpha=Get Sleeve L Correction() -> (Sleeve L Correction:double), Translation="-2,0,0", Rotation="0,0,0", Scale="1,1,1") {
    translationMode=BMM_Additive
    translationSpace=BCS_BoneSpace
    alphaBoolBlend {blendOption=Linear}
    boneToModify {boneName=sleeve_L}
    translation {x=-2}
   }
   n30 = "Transform (Modify) Bone - Bone: sleeve_R"(
    ComponentPose=@n29
    Alpha=Get Sleeve R Correction() -> (Sleeve R Correction:float)
    Translation="3.5,0,-0.35"
    Rotation="0,0,0"
    Scale="1,1,1"
   ) {
    translationMode=BMM_Additive
    translationSpace=BCS_BoneSpace
    alphaBoolBlend {blendOption=Linear}
    boneToModify {boneName=sleeve_R}
    translation {x=3.5, z=-0.35}
   }
   n31 = "Transform (Modify) Bone - Bone: bag_fr"(ComponentPose=@n30, Alpha=@n17, Translation="-2,0,0.2", Rotation="-7,0,0", Scale="1,1,1") {
    rotationMode=BMM_Additive
    rotationSpace=BCS_BoneSpace
    translationMode=BMM_Additive
    translationSpace=BCS_BoneSpace
    alphaBoolBlend {blendOption=Linear}
    boneToModify {boneName=bag_fr}
    rotation {pitch=-7}
    translation {x=-2, z=0.2}
   }
   n32 = "Transform (Modify) Bone - Bone: bag_01_L"(ComponentPose=@n31, Alpha=@n17, Translation="0,0,0.7", Rotation="0,0,-5", Scale="1,1,1") {
    rotationMode=BMM_Additive
    rotationSpace=BCS_BoneSpace
    translationMode=BMM_Additive
    translationSpace=BCS_BoneSpace
    alphaBoolBlend {blendOption=Linear}
    boneToModify {boneName=bag_01_L}
    rotation {roll=-5}
    translation {z=0.7}
   }
   n33 = "Transform (Modify) Bone - Bone: belt_L"(ComponentPose=@n32, Alpha=Get Belt correction() -> (Belt correction:double), Translation="-3,0,0", Rotation="0,-15,0", Scale="1,1,1") {
    rotationMode=BMM_Additive
    rotationSpace=BCS_BoneSpace
    translationMode=BMM_Additive
    translationSpace=BCS_BoneSpace
    alphaBoolBlend {blendOption=Linear}
    boneToModify {boneName=belt_L}
    rotation {yaw=-15}
    translation {x=-3}
   }
   n34 = "Transform (Modify) Bone - Bone: vest_bk_04"(ComponentPose=@n33, Alpha=@n16, Translation="0,-2.5,0", Rotation="0,0,22", Scale="1,1,1") {
    rotationMode=BMM_Additive
    rotationSpace=BCS_BoneSpace
    translationMode=BMM_Additive
    translationSpace=BCS_BoneSpace
    alphaBoolBlend {blendOption=Linear}
    alphaScaleBiasClamp {interpSpeedDecreasing=7, interpSpeedIncreasing=0}
    boneToModify {boneName=vest_bk_04}
    rotation {roll=22}
    translation {y=-2.5}
   }
   n35 = Get body rotation to the right() -> (body rotation to the right:double)
   n36 = "Transform (Modify) Bone - Bone: vest_bk_04"(ComponentPose=@n34, Alpha=@n35, Translation="0,-2.5,0", Rotation="0,0,-22", Scale="1,1,1") {
    rotationMode=BMM_Additive
    rotationSpace=BCS_BoneSpace
    translationMode=BMM_Additive
    translationSpace=BCS_BoneSpace
    alphaBoolBlend {blendOption=Linear}
    alphaScaleBiasClamp {interpSpeedDecreasing=7, interpSpeedIncreasing=0}
    boneToModify {boneName=vest_bk_04}
    rotation {roll=-22}
    translation {y=-2.5}
   }
   n37 = Get Vest front correction() -> (Vest front correction:double)
   n38 = "Transform (Modify) Bone - Bone: vest_fr_04"(ComponentPose=@n36, Alpha=@n37, Translation="0,1,0", Rotation="0,0,0", Scale="1,1,1") {
    translationMode=BMM_Additive
    translationSpace=BCS_BoneSpace
    alphaBoolBlend {blendOption=Linear}
    alphaScaleBiasClamp {interpSpeedDecreasing=7, interpSpeedIncreasing=0}
    boneToModify {boneName=vest_fr_04}
    translation {y=1}
   }
   n39 = "Transform (Modify) Bone - Bone: vest_bk_05"(ComponentPose=@n38, Alpha=@n16, Translation="0,-3,0", Rotation="0,0,22", Scale="1,1,1") {
    rotationMode=BMM_Additive
    rotationSpace=BCS_BoneSpace
    translationMode=BMM_Additive
    translationSpace=BCS_BoneSpace
    alphaBoolBlend {blendOption=Linear}
    alphaScaleBiasClamp {interpSpeedDecreasing=7, interpSpeedIncreasing=0}
    boneToModify {boneName=vest_bk_05}
    rotation {roll=22}
    translation {y=-3}
   }
   n40 = "Transform (Modify) Bone - Bone: vest_bk_05"(ComponentPose=@n39, Alpha=@n35, Translation="0,-3,0", Rotation="0,0,-22", Scale="1,1,1") {
    rotationMode=BMM_Additive
    rotationSpace=BCS_BoneSpace
    translationMode=BMM_Additive
    translationSpace=BCS_BoneSpace
    alphaBoolBlend {blendOption=Linear}
    alphaScaleBiasClamp {interpSpeedDecreasing=7, interpSpeedIncreasing=0}
    boneToModify {boneName=vest_bk_05}
    rotation {roll=-22}
    translation {y=-3}
   }
   n41 = "Transform (Modify) Bone - Bone: vest_fr_05"(ComponentPose=@n40, Alpha=@n37, Translation="0,1,0", Rotation="0,0,0", Scale="1,1,1") {
    translationMode=BMM_Additive
    translationSpace=BCS_BoneSpace
    alphaBoolBlend {blendOption=Linear}
    alphaScaleBiasClamp {interpSpeedDecreasing=7, interpSpeedIncreasing=0}
    boneToModify {boneName=vest_fr_05}
    translation {y=1}
   }
   n42 = "Transform (Modify) Bone - Bone: buckle_L"(
    ComponentPose=@n41
    Alpha=Get Clavicle L Backward Movement() -> (Clavicle L Backward Movement:float)
    Translation=Get Buckle L Side Movement() -> (Buckle L Side Movement:Vector)
    Rotation=@n15
    Scale="1,1,1"
   ) {
    rotationMode=BMM_Additive
    rotationSpace=BCS_BoneSpace
    translationMode=BMM_Additive
    translationSpace=BCS_BoneSpace
    alphaBoolBlend {blendOption=Linear}
    alphaScaleBiasClamp {interpSpeedDecreasing=0.1}
    boneToModify {boneName=buckle_L}
   }
   n43 = "Transform (Modify) Bone - Bone: buckle_R"(
    ComponentPose=@n42
    Alpha=Get Clavicle R Backward Movement() -> (Clavicle R Backward Movement:float)
    Translation=Get Buckle R Side Movement() -> (Buckle R Side Movement:Vector)
    Rotation=@n15
    Scale="1,1,1"
   ) {
    rotationMode=BMM_Additive
    rotationSpace=BCS_BoneSpace
    translationMode=BMM_Additive
    translationSpace=BCS_BoneSpace
    alphaBoolBlend {blendOption=Linear}
    alphaScaleBiasClamp {interpSpeedDecreasing=0.1}
    boneToModify {boneName=buckle_R}
   }
   n44 = "Transform (Modify) Bone - Bone: buckle_R"(
    ComponentPose=@n43
    Alpha=Get Clavicle R forward movement() -> (Clavicle R forward movement:float)
    Translation="3.439,1,0"
    Rotation="0,0,0"
    Scale="1,1,1"
   ) {
    rotationMode=BMM_Additive
    rotationSpace=BCS_BoneSpace
    translationMode=BMM_Additive
    translationSpace=BCS_BoneSpace
    alphaBoolBlend {blendOption=Linear}
    alphaScaleBiasClamp {interpSpeedDecreasing=0.1}
    boneToModify {boneName=buckle_R}
    translation=<3.439 1>
   }
   n45 = "Transform (Modify) Bone - Bone: buckle_L"(
    ComponentPose=@n44
    Alpha=Get Clavicle L forward movement() -> (Clavicle L forward movement:float)
    Translation="-3.439,-1,0"
    Rotation="0,0,0"
    Scale="1,1,1"
   ) {
    rotationMode=BMM_Additive
    rotationSpace=BCS_BoneSpace
    translationMode=BMM_Additive
    translationSpace=BCS_BoneSpace
    alphaBoolBlend {blendOption=Linear}
    alphaScaleBiasClamp {interpSpeedDecreasing=0.1}
    boneToModify {boneName=buckle_L}
    translation=<-3.439 -1>
   }
   n46 = "Transform (Modify) Bone - Bone: harness_L"(
    ComponentPose=@n45
    Alpha=Get Clavicle L upward movement()
    Translation=Get Harness L moving up() -> (Harness L moving up:Vector)
    Rotation=Get Harness L Upper Rotation() -> (Harness L Upper Rotation:Rotator)
    Scale="1,1,1"
   ) {
    rotationMode=BMM_Additive
    rotationSpace=BCS_BoneSpace
    translationMode=BMM_Additive
    translationSpace=BCS_BoneSpace
    alphaBoolBlend {blendOption=Linear}
    alphaScaleBiasClamp {interpSpeedDecreasing=20, interpSpeedIncreasing=0}
    boneToModify {boneName=harness_L}
    rotation {pitch=20}
    translation {x=6.559, z=-3.3}
   }
   n47 = "Transform (Modify) Bone - Bone: harness_R"(
    ComponentPose=@n46
    Alpha=Get Clavicle R upward movement()
    Translation=Get Harness R moving up() -> (Harness R moving up:Vector)
    Rotation=Get Harness R Upper Rotation() -> (Harness R Upper Rotation:Rotator)
    Scale="1,1,1"
   ) {
    rotationMode=BMM_Additive
    rotationSpace=BCS_BoneSpace
    translationMode=BMM_Additive
    translationSpace=BCS_BoneSpace
    alphaBoolBlend {blendOption=Linear}
    alphaScaleBiasClamp {interpSpeedDecreasing=20, interpSpeedIncreasing=0}
    boneToModify {boneName=harness_R}
    rotation {pitch=20}
    translation=<-6.559 -1 3.3>
   }
   n48 = Get Harness Side Rotation() -> (Harness Side Rotation:Rotator)
   n49 = "Transform (Modify) Bone - Bone: harness_L"(
    ComponentPose=@n47
    Alpha=Get Clavicle L Backward Movement()
    Translation=Get Harness L Side Movement() -> (Harness L Side Movement:Vector)
    Rotation=@n48
    Scale="1,1,1"
   ) {
    rotationMode=BMM_Additive
    rotationSpace=BCS_BoneSpace
    translationMode=BMM_Additive
    translationSpace=BCS_BoneSpace
    alphaBoolBlend {blendOption=Linear}
    alphaScaleBiasClamp {interpSpeedDecreasing=20, interpSpeedIncreasing=0}
    boneToModify {boneName=harness_L}
    rotation {yaw=10}
   }
   n50 = "Transform (Modify) Bone - Bone: harness_R"(
    ComponentPose=@n49
    Alpha=Get Clavicle R Backward Movement()
    Translation=Get Harness R Side Movement() -> (Harness R Side Movement:Vector)
    Rotation=@n48
    Scale="1,1,1"
   ) {
    rotationMode=BMM_Additive
    rotationSpace=BCS_BoneSpace
    translationMode=BMM_Additive
    translationSpace=BCS_BoneSpace
    alphaBoolBlend {blendOption=Linear}
    alphaScaleBiasClamp {interpSpeedDecreasing=20, interpSpeedIncreasing=0}
    boneToModify {boneName=harness_R}
    rotation {yaw=10}
   }
   Save cached pose 'Jacket Cashe'(Pose=Component To Local(ComponentPose=@n50)) {cacheName=Jacket Cashe}
   Reroute Node(InputPin:Wildcard) -> (OutputPin:Wildcard)
  ]}
 EventGraph {nodes [
   Event OnCharacterCustomizationLoaded()
   > n1 = Set SkelMesh(SkelMesh=GetPartSkeletalMeshComponent())
   > n2 = Get Dynamic Material Instance(SkeletalMeshComponent=@n1)
   > Set Mat Instance Dynamic(Mat Instance Dynamic=@n2)
   > n3 = GetOrCreateDynamicOverlayMaterial(MeshComponent=@n1, MaterialName=None)
   > Set Overlay Mat Instance Dynamic(Overlay Mat Instance Dynamic=@n3)
   Event BlueprintInitializeAnimation()
   > Set SkelMesh(SkelMesh=GetOwnerSkeletalMeshComponent())
   n4 = TryGetPawnOwner()
   > Is Valid(InputObject=@n4)
   > n5 = Cast To FortPlayerPawnAthena(Object=@n4)
   > Set PlayerPawn(PlayerPawn:Fort Player Pawn Athena Object Reference=@n5) -> (Output_Get:Fort Player Pawn Athena Object Reference)
   n6 = Event BlueprintUpdateAnimation()
   > Set Delta Time X(Delta Time X=@n6.DeltaTimeX)
   > n7 = Sequence!(execute) -> (then_0, then_1, then_2:Exec, then_3:Exec)
   n8 = Get bIsSkyDiving()
   > Set Harness L Side Movement(Harness L Side Movement:Vector=SelectVector(A="2,0,0", B="0, 0, 0", bPickA=@n8)) -> (Output_Get:Vector)
   > Set Harness R Side Movement(Harness R Side Movement:Vector=SelectVector(A="-2,0,0", B="0, 0, 0", bPickA=@n8)) -> (Output_Get:Vector)
   > Set Harness Side Rotation(Harness Side Rotation:Rotator=SelectRotator!(A:Rotator="0,10,0", B:Rotator="0, 0, 0", bPickA:bool=@n8) -> (ReturnValue:Rotator)) -> (Output_Get:Rotator)
   > Set Harness L moving up(Harness L moving up:Vector=SelectVector(A="2,0,0", B="6.559,0,-3", bPickA=@n8)) -> (Output_Get:Vector)
   > Set Harness L Upper Rotation(Harness L Upper Rotation:Rotator=SelectRotator(A="0, 0, 0", B="20,0,0", bPickA=@n8)) -> (Output_Get:Rotator)
   > Set Harness R moving up(Harness R moving up:Vector=SelectVector(A="-2,0,0", B="-6.559,-1,3", bPickA=@n8)) -> (Output_Get:Vector)
   > Set Harness R Upper Rotation(Harness R Upper Rotation:Rotator=SelectRotator(A="0, 0, 0", B="20,2,0", bPickA=@n8)) -> (Output_Get:Rotator)
   > Set Buckle Side Rotation(Buckle Side Rotation:Rotator=SelectRotator(A="0,-1,0", B="0, 0, 0", bPickA=@n8)) -> (Output_Get:Rotator)
   > Set Buckle L Side Movement(Buckle L Side Movement:Vector=SelectVector(A="4,0,1", B="0, 0, 0", bPickA=@n8)) -> (Output_Get:Vector)
   > Set Buckle R Side Movement(Buckle R Side Movement:Vector=SelectVector(A="-4,0,-1", B="0, 0, 0", bPickA=@n8)) -> (Output_Get:Vector)
   > Set Collar BK MSTR L Upper Rotation(Collar BK MSTR L Upper Rotation:Rotator=SelectRotator(A="0, 0, 0", B="-13.883,8.106,-26.971", bPickA=@n8)) -> (Output_Get:Rotator)
   > Set Collar BK MSTR R Upper Rotation(Collar BK MSTR R Upper Rotation:Rotator=SelectRotator(A="0, 0, 0", B="-12.745,-3.542,47.522", bPickA=@n8)) -> (Output_Get:Rotator)
   n9 = Get SkelMesh()
   Set Sleeve L Correction(
    execute=@n7.then_1
    Sleeve L Correction:double=float + float(
     A=GetAlphaFromReferenceTransformDelta(SkeletalMeshComponent=@n9, BoneName=hand_l, BaseName=lowerarm_l, DeltaComponentType=EulerY, InputRange=(X=0,Y=50), TwistAxis="1,0,0")
     B=GetAlphaFromReferenceTransformDelta(SkeletalMeshComponent=@n9, BoneName=hand_l, BaseName=lowerarm_l, DeltaComponentType=EulerY, InputRange=(X=0,Y=-30), TwistAxis="1,0,0")
    )
   ) -> (Output_Get:double)
   n10 = Get SkelMesh()
   > Set Sleeve R Correction(Sleeve R Correction:float=float + float(
     A=GetAlphaFromReferenceTransformDelta(SkeletalMeshComponent=@n10, BoneName=hand_r, BaseName=lowerarm_r, DeltaComponentType=EulerY, InputRange=(X=0,Y=66), TwistAxis="1,0,0")
     B=GetAlphaFromReferenceTransformDelta(
      SkeletalMeshComponent=@n10
      BoneName=hand_r
      BaseName=lowerarm_r
      DeltaComponentType=EulerY
      InputRange=(X=0,Y=-27)
      OutputRange=(X=0,Y=0.3)
      TwistAxis="1,0,0"
     )
    )) -> (Output_Get:float)
   n11 = Get SkelMesh()
   n12 = NOT Boolean(A=Get bIsMoving2D())
   n13 = Get bIsCrouching()
   > Set Belt correction(Belt correction:double=SelectFloat(
     A=GetAlphaFromReferenceTransformDelta(SkeletalMeshComponent=@n11, BoneName=thigh_l, BaseName=pelvis, InputRange=(X=-24,Y=-103), TwistAxis="1,0,0")
     B=0
     bPickA=AND Boolean(A=@n13, B=@n12)
    )) -> (Output_Get:double)
   n14 = NOT Boolean(A=Get bIsMoving2D())
   n15 = Get bIsCrouching()
   > Set Crouching standing(Crouching standing:double=SelectFloat(A=1, B=0, bPickA=AND Boolean(A=@n15, B=@n14))) -> (Output_Get:double)
   > n16 = Gravity Override State Function DEPRECATED(CharacterPartAnimInstance=Self-Reference(), GameTimeInSeconds=GetGameTimeInSeconds(), GravityOverrideParams=Get Gravity Override Params())
   > Set Gravity Override(Gravity Override=@n16)
   n17 = SelectVector(A="0.1,0.1,0.1", B="0.4,0.4,0.4", bPickA=Get bIsSkyDiving())
   > Set ComponentLinearAccScale(ComponentLinearAccScale:Vector=@n17) -> (Output_Get:Vector)
   > Set ComponentLinearVelScale(ComponentLinearVelScale:Vector=@n17) -> (Output_Get:Vector)
   n18 = Get SkelMesh()
   Set Clavicle L upward movement(
    execute=@n7.then_2
    Clavicle L upward movement:double=GetAlphaFromReferenceTransformDelta(SkeletalMeshComponent=@n18, BoneName=clavicle_l, BaseName=spine_05, DeltaComponentType=EulerY, InputRange=(X=7,Y=47), TwistAxis="1,0,0")
   ) -> (Output_Get:double)
   > Set Clavicle R upward movement(Clavicle R upward movement:float=GetAlphaFromReferenceTransformDelta(SkeletalMeshComponent=@n18, BoneName=clavicle_r, BaseName=spine_05, DeltaComponentType=EulerY, InputRange=(X=7,Y=47), TwistAxis="1,0,0")) -> (Output_Get:float)
   > Set Neck 01 Rotation L(Neck 01 Rotation L:double=GetAlphaFromReferenceTransformDelta(SkeletalMeshComponent=@n18, BoneName=neck_01, BaseName=spine_05, DeltaComponentType=EulerX, InputRange=(X=0,Y=-28), TwistAxis="1,0,0")) -> (Output_Get:double)
   > Set Neck 01 Rotation R(Neck 01 Rotation R:float=GetAlphaFromReferenceTransformDelta(SkeletalMeshComponent=@n18, BoneName=neck_01, BaseName=spine_05, DeltaComponentType=EulerX, InputRange=(X=0,Y=28), TwistAxis="1,0,0")) -> (Output_Get:float)
   n19 = Get SkelMesh()
   > Set Clavicle L Backward Movement(Clavicle L Backward Movement:float=GetAlphaFromReferenceTransformDelta(SkeletalMeshComponent=@n19, BoneName=clavicle_l, BaseName=spine_05, InputRange=(X=-5,Y=-22), TwistAxis="1,0,0")) -> (Output_Get:float)
   > Set Clavicle R Backward Movement(Clavicle R Backward Movement:float=GetAlphaFromReferenceTransformDelta(SkeletalMeshComponent=@n19, BoneName=clavicle_r, BaseName=spine_05, InputRange=(X=-5,Y=-22), TwistAxis="1,0,0")) -> (Output_Get:float)
   n20 = GetAlphaFromReferenceTransformDelta(SkeletalMeshComponent=@n19, BoneName=clavicle_r, BaseName=spine_05, InputRange=(X=0,Y=23), TwistAxis="1,0,0")
   n21 = GetAlphaFromReferenceTransformDelta(SkeletalMeshComponent=@n19, BoneName=clavicle_l, BaseName=spine_05, InputRange=(X=0,Y=23), TwistAxis="1,0,0")
   > Set Vest front correction(Vest front correction:double=float + float(A=@n21, B=@n20)) -> (Output_Get:double)
   > Set Clavicle R forward movement(Clavicle R forward movement:float=@n20) -> (Output_Get:float)
   > Set Clavicle L forward movement(Clavicle L forward movement:float=@n21) -> (Output_Get:float)
   n22 = Get SkelMesh()
   > Set body rotation to the left(body rotation to the left:double=float + float(
     A=GetAlphaFromReferenceTransformDelta(
      SkeletalMeshComponent=@n22
      BoneName=spine_05
      BaseName=spine_04
      DeltaComponentType=EulerX
      InputRange=(X=0,Y=40)
      OutputRange=(X=0,Y=0.5)
      TwistAxis="1,0,0"
     )
     B=GetAlphaFromReferenceTransformDelta(SkeletalMeshComponent=@n22, BoneName=clavicle_l, BaseName=spine_05, InputRange=(X=-5,Y=-22), OutputRange=(X=0,Y=0.5), TwistAxis="1,0,0")
    )) -> (Output_Get:double)
   > Set body rotation to the right(body rotation to the right:double=float + float(
     A=GetAlphaFromReferenceTransformDelta(
      SkeletalMeshComponent=@n22
      BoneName=spine_05
      BaseName=spine_04
      DeltaComponentType=EulerX
      InputRange=(X=0,Y=-40)
      OutputRange=(X=0,Y=0.5)
      TwistAxis="1,0,0"
     )
     B=GetAlphaFromReferenceTransformDelta(SkeletalMeshComponent=@n22, BoneName=clavicle_r, BaseName=spine_05, InputRange=(X=-5,Y=-22), OutputRange=(X=0,Y=0.5), TwistAxis="1,0,0")
    )) -> (Output_Get:double)
   n23 = Get SkelMesh()
   > Set Bag correction when crouching(Bag correction when crouching:double=SelectFloat(
     A=GetAlphaFromReferenceTransformDelta(SkeletalMeshComponent=@n23, BoneName=spine_02, BaseName=pelvis, InputRange=(X=8.3,Y=18), TwistAxis="1,0,0")
     B=0
     bPickA=Get bIsCrouching()
    )) -> (Output_Get:double)
   > Set Head Pitch(Head Pitch:float=GetAlphaFromReferenceTransformDelta(SkeletalMeshComponent=@n23, BoneName=head, BaseName=spine_05, InputRange=(X=-10,Y=19), TwistAxis="1,0,0")) -> (Output_Get:float)
   Is Valid(exec=@n7.then_3, InputObject=Get PlayerPawn() -> (PlayerPawn:Fort Player Pawn Athena Object Reference))
   > Set WPO Globals Function(
    Material=Get Mat Instance Dynamic()
    Overlay Material=Get Overlay Mat Instance Dynamic()
    Skeletal Mesh=Get SkelMesh()
    "Local Noise Transform  Location"="Get Local Noise Transform  Location"() -> ("Local Noise Transform  Location":Vector)
    Local Noise Offset=Get Local Noise Offset()
   )
   > Set WPO Regional Function!(
    execute
    Material:Material Instance Dynamic Object Reference=Get Mat Instance Dynamic()
    Overlay Material:Material Instance Dynamic Object Reference=Get Overlay Mat Instance Dynamic()
    Skeletal Mesh:Skeletal Mesh Component Object Reference=Get SkelMesh()
    Velocity Transform:Name=spine_05
    Velocity Interp Speed:double=Get Velocity Interp Speed()
    Collision Transform:Name=pelvis
    Collision Transform Axis:EAxis Enum
    Invert Collision Transform Axis:bool=false
    Delta Time X:double=Get Delta Time X()
    Previous Velocity Transform Location:"Vector (by ref)"=Get R1 Velocity Transform Location()
    Previous Velocity Transform Velocity:"Vector (by ref)"=Get R1 Velocity Transform Velocity()
    Region:int=1
   ) -> (then)
   > n24 = Sequence!(execute) -> (then_0, then_1)
   > n25 = Branch(Condition=Get bIsSkydiveFloating())
   > n26 = Set WPO Pho Wind Intensity(WPO Pho Wind Intensity=1)
   n27 = Get Mat Instance Dynamic()
   SetScalarParameterValue(execute=[@n28.then, @n26.then], self=@n27, ParameterName=ClothSim_WindStrength, Value=Get WPO Pho Wind Intensity())
   n28 = Set WPO Pho Wind Intensity(execute=@n25.else, WPO Pho Wind Intensity=0)
   SetScalarParameterValue(
    execute=@n24.then_1
    self=@n27
    ParameterName=ClothSim_R1_DragStrength
    Value=SelectFloat(A=0.6, B=1.5, bPickA=Get bIsInVehicle() -> (bIsInVehicle:bool))
   )
   > SetScalarParameterValue(self=@n27, ParameterName=ClothSim_R1_TurbulenceStrength, Value=SelectFloat(A=0.5, B=1.4, bPickA=Get bIsInVehicle()))
  ]}
]
variables [
 Bag correction when crouching {type=number}
 Belt correction {type=number}
 body rotation to the left {type=number}
 body rotation to the right {type=number}
 Buckle L Side Movement {type=Vector}
 Buckle R Side Movement {type=Vector}
 Buckle Side Rotation {type=Rotator}
 Clavicle L Backward Movement {type=number}
 Clavicle L forward movement {type=number}
 Clavicle L upward movement {type=number}
 Clavicle R Backward Movement {type=number}
 Clavicle R forward movement {type=number}
 Clavicle R upward movement {type=number}
 Collar BK MSTR L Upper Rotation {type=Rotator}
 Collar BK MSTR R Upper Rotation {type=Rotator}
 ComponentLinearAccScale {type=Vector}
 ComponentLinearVelScale {type=Vector}
 Crouching standing {type=number}
 Delta Time X {type=number}
 Gravity Override {type=Vector}
 Gravity Override Params {
  type=GravityOverrideParamsStruct
  default {
   skydive_FloatingDown {gravityOverride=<0 -1200 100>, windFrequency=<1 1 1>}
   skydive_GoingUp {gravityOverride=<0 -1200 100>}
   skydive_Parachuting {windFrequency=<1.5 2 3>}
  }
 }
 Harness L moving up {type=Vector}
 Harness L Side Movement {type=Vector}
 Harness L Upper Rotation {type=Rotator}
 Harness R moving up {type=Vector}
 Harness R Side Movement {type=Vector}
 Harness R Upper Rotation {type=Rotator}
 Harness Side Rotation {type=Rotator}
 Head Pitch {type=number}
 Local Noise Offset {type=Vector}
 "Local Noise Transform  Location" {type=Vector}
 Mat Instance Dynamic {type=MaterialInstanceDynamic}
 Neck 01 Rotation L {type=number}
 Neck 01 Rotation R {type=number}
 Overlay Mat Instance Dynamic {type=MaterialInstanceDynamic}
 PlayerPawn {type=FortPlayerPawnAthena}
 R1 Velocity Transform Location {type=Vector}
 R1 Velocity Transform Velocity {type=Vector}
 SkelMesh {type=SkeletalMeshComponent}
 Sleeve L Correction {type=number}
 Sleeve R Correction {type=number}
 Velocity Interp Speed {type=number, default=30}
 Vest front correction {type=number}
 WPO Pho Wind Intensity {type=number}
]
